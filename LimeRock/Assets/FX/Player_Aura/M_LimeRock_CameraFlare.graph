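# Translucent, unlit, two-sided material for Niagara mesh particles, named as the camera flare in LimeRock's player aura effects.
# A panning turbulence texture distorts two scrolling noise layers: powered caustics tinted by one color, plus plasma blending two others.
# The sum is multiplied by a static mask and particle color, then passed through MF_FX_EmissiveAdjustments.
# Opacity multiplies opposing powered V gradients, forming a band, by the plasma layer and particle alpha.
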Material /LimeRock/Assets/FX/Player_Aura/M_LimeRock_CameraFlare
graph {
 properties {
  bUsedWithNiagaraMeshParticles=true
  blendMode=BLEND_Translucent
  pixelDepthOffsetMode=PDOM_Legacy
  shadingModel=MSM_Unlit
  twoSided=true
  usedShadingModels=Unlit
 }
 nodes [
  n1 = TextureSampleParameter2D(UVs=TextureCoordinate()) {
   group=Emissive
   parameterName=EmissiveMask_Texture
   samplerType=SAMPLERTYPE_LinearGrayscale
   texture=/Game/Global/Textures/Default/Caustics/T_Nosie_Caustics_Grayscale
  }
  n2 = Multiply(
   A=TextureCoordinate() {uTiling=2, vTiling=2}
   B=AppendVector(
    A=ScalarParameter() {defaultValue=1, group=Emissive, parameterName=Emissive_Freq_U_01}
    B=ScalarParameter() {defaultValue=1, group=Emissive, parameterName=Emissive_Freq_V_01}
   )
  )
  n3 = TextureCoordinate() {vTiling=2}
  n4 = AppendVector(
   A=ScalarParameter() {defaultValue=1, group=Turb, parameterName=Turb_Freq_U}
   B=ScalarParameter() {defaultValue=1, group=Turb, parameterName=Turb_Freq_V}
  )
  n5 = ScalarParameter() {group=Turb, parameterName=Turb_Speed_U}
  n6 = ScalarParameter() {defaultValue=0.2, group=Turb, parameterName=Turb_Speed_V}
  n7 = TextureSampleParameter2D(UVs=Panner(Coordinate=Multiply(A=@n3, B=@n4), Speed=AppendVector(A=@n5, B=@n6)) {speedY=0.2}) {
   group=Turb
   parameterName=Turb_Texture
   samplerType=SAMPLERTYPE_LinearGrayscale
   texture=/Game/Effects/Fort_Effects/Emotes/UltraLight/Texture/T_DirectionalNoise_Greyscale
  }
  n8 = Multiply(A=@n7.R, B=ScalarParameter() {defaultValue=0.05, group=Turb, parameterName=Turb_Amp}) {constB=0.05}
  n9 = ScalarParameter() {group=Emissive, parameterName=Emissive_Speed_U_01}
  n10 = ScalarParameter() {defaultValue=-1, group=Emissive, parameterName=Emissive_Speed_V_01}
  n11 = TextureSampleParameter2D(UVs=Panner(Coordinate=Add(A=@n2, B=@n8) {constB=0}, Speed=AppendVector(A=@n9, B=@n10)) {speedY=-1}) {
   group=Emissive
   parameterName=Emissive_Texture_01
   samplerType=SAMPLERTYPE_LinearGrayscale
   texture=&T_Nosie_Caustics_Grayscale
  }
  n12 = ScalarParameter() {defaultValue=2, group=Emissive, parameterName=Emissive_Pow_01}
  n13 = Multiply(
   A=TextureCoordinate() {uTiling=3}
   B=AppendVector(
    A=ScalarParameter() {defaultValue=1, group=Emissive, parameterName=Emissive_Freq_U_02}
    B=ScalarParameter() {defaultValue=1, group=Emissive, parameterName=Emissive_Freq_V_02}
   )
  )
  n14 = ScalarParameter() {group=Emissive, parameterName=Emissive_Speed_U_02}
  n15 = ScalarParameter() {defaultValue=-0.3, group=Emissive, parameterName=Emissive_Speed_V_02}
  n16 = TextureSampleParameter2D(UVs=Panner(Coordinate=Add(A=@n13, B=@n8) {constB=0}, Speed=AppendVector(A=@n14, B=@n15)) {speedY=-0.3}) {
   group=Emissive
   parameterName=Emissive_Texture_02
   samplerType=SAMPLERTYPE_LinearGrayscale
   texture=/Game/Effects/Fort_Effects/Emotes/UltraLight/Texture/T_Plasma_C_Greyscale
  }
  n17 = Add(
   A=Multiply(A=VectorParameter() {group=Emissive, parameterName=Color_A, defaultValue=<rgba 0 1.9534 2 1>}.RGB, B=Power(Base=@n11.R, Exp=@n12))
   B=LinearInterpolate(
    A=VectorParameter() {group=Emissive, parameterName=Color_B, defaultValue=<rgba 0 0.9313 1 1>}.RGB
    B=VectorParameter() {group=Emissive, parameterName=Color_C, defaultValue=<rgba 0 1.5 0.2228 1>}.RGB
    Alpha=@n16.R
   )
  )
  n18 = ParticleColor()
  MaterialFunctionCall(Emissive=Multiply(A=@n1.R, B=Multiply(A=@n17, B=@n18.RGB))) {materialFunction=&MF_FX_EmissiveAdjustments}
  n19 = TextureCoordinate()
  n20 = Power(
   Base=OneMinus(None=ComponentMask(None=@n19) {g=true})
   Exp=ScalarParameter() {defaultValue=2, group=Opacity, parameterName=Opacity_Inverted_G_Power}
  )
  n21 = ComponentMask(None=TextureCoordinate()) {g=true}
  n22 = ScalarParameter() {defaultValue=25, group=Opacity, parameterName=Opacity_G_Power}
  n23 = Saturate(None=Multiply(A=Power(Base=@n21, Exp=@n22) {constExponent=25}, B=ScalarParameter() {defaultValue=20, group=Opacity, parameterName=Opacity_Bottom_Amp}) {constB=20})
  n24 = Saturate(None=Multiply(A=Multiply(A=@n20, B=@n23), B=ScalarParameter() {defaultValue=20, group=Opacity, parameterName=Opacity_Amp}) {constB=20})
  Multiply(A=Multiply(A=@n18.A, B=Multiply(A=@n16.R, B=@n24)))
 ]
}
